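# Sound cue for Section 02 (tac9200) of the PistolCanary event audio, apparently a tentacle sound given its attenuation asset.
# A single wave player feeds the output directly, with no randomization, mixing or modulation nodes.
# It plays at reduced volume in the Fort_Events sound class, spatialized through the shared PistolCanary tentacle attenuation with a long maximum distance.
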SoundCue /PistolCanary_Event_Audio/Assets/Section02/tacaud9200/SC_PistolCanary_Section02_tac9200
root=WavePlayer_1
nodes [{id=WavePlayer_1}]
properties {
 attenuationSettings=/PistolCanary_Event_Audio/Assets/Settings/ATT_PistolCanary_Tentacle
 maxDistance=23000
 soundClassObject=&Fort_Events
 volumeMultiplier=0.6627
}
